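# Physics asset for the FaunaPike offhand pickaxe axe, defining capsule bodies for its dangling chain, handle and the wielder's legs.
# Six dyn_chain bodies link in series from the MSTR_chain root through constraints with per-joint swing limits and zero-degree twist limits.
# Handle, thigh and calf capsules have no constraints, apparently serving as collision shapes for the swinging chain.
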PhysicsAsset /BRCosmetics/Weapons/FORT_Melee/Pickaxe_FaunaPike_offhand/Meshes/FaunaPike_Offhand_Axe_Physics
bodies {
 MSTR_chain [{shapeType=Capsule, length=0.505, radius=0.505, shapeName=MSTR_chain_capsule, rotation=<rot 0 -90 0>}]
 Melee_TwoHanded_Handle [{shapeType=Capsule, length=68.4892, radius=20.5601, center=<0 0 41.303>}]
 calf_l [{shapeType=Capsule, length=50, radius=8, center=<-30 -2 0>, rotation=<rot 90 0 0>}]
 calf_r [{shapeType=Capsule, length=50, radius=8, center=<30 2 0>, rotation=<rot 90 0 0>}]
 dyn_chain_01 [{shapeType=Capsule, length=4.7967, radius=2.7026, shapeName=dyn_chain_01_capsule, center=<4 0 0>, rotation=<rot 90 0 0>}]
 dyn_chain_02 [{shapeType=Capsule, length=3.0329, radius=0.7446, shapeName=dyn_chain_02_capsule, center=<2 0 0>, rotation=<rot 90 0 0>}]
 dyn_chain_03 [{shapeType=Capsule, length=2.8503, radius=0.7815, shapeName=dyn_chain_03_capsule, center=<1.8 0 0>, rotation=<rot 90 0 0>}]
 dyn_chain_04 [{shapeType=Capsule, length=2.9286, radius=0.9183, shapeName=dyn_chain_04_capsule, center=<1.3 0 0>, rotation=<rot 90 0 0>}]
 dyn_chain_05 [{shapeType=Capsule, length=4.0559, radius=0.9152, shapeName=dyn_chain_05_capsule, center=<2.6 0 0>, rotation=<rot 90 0 0>}]
 dyn_chain_06 [{shapeType=Capsule, length=5.9734, radius=5.0382, shapeName=dyn_chain_06_capsule, center=<4.6 -1 0>, rotation=<rot -0.0909 112.8752 -89.8352>}]
 thigh_l [{shapeType=Capsule, length=40, radius=8, center=<-25 0 0>, rotation=<rot 90 0 0>}]
 thigh_r [{shapeType=Capsule, length=40, radius=8, center=<25 0 0>, rotation=<rot 90 0 0>}]
}
constraints [
 {bone1Name=dyn_chain_01, bone2Name=MSTR_chain, swing1LimitDegrees=2, swing2LimitDegrees=75, twistLimitDegrees=0}
 {bone1Name=dyn_chain_02, bone2Name=dyn_chain_01, swing1LimitDegrees=80, swing2LimitDegrees=45, twistLimitDegrees=0}
 {bone1Name=dyn_chain_03, bone2Name=dyn_chain_02, swing1LimitDegrees=45, swing2LimitDegrees=45, twistLimitDegrees=0}
 {bone1Name=dyn_chain_04, bone2Name=dyn_chain_03, swing1LimitDegrees=45, swing2LimitDegrees=45, twistLimitDegrees=0}
 {bone1Name=dyn_chain_05, bone2Name=dyn_chain_04, swing1LimitDegrees=45, swing2LimitDegrees=45, twistLimitDegrees=0}
 {bone1Name=dyn_chain_06, bone2Name=dyn_chain_05, swing1LimitDegrees=10, swing2LimitDegrees=50, twistLimitDegrees=0}
]
