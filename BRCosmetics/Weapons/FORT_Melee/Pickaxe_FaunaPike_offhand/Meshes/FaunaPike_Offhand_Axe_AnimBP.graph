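# Animation Blueprint for the FaunaPike offhand pickaxe that adds physics motion to the weapon mesh.
# Each update copies world transforms of pelvis, thighs and calves from the attach-parent body mesh.
# Modify Bone nodes apply these to matching bones ahead of a RigidBody simulation, likely serving as body colliders.
# bRigidBodyEnabled selects that result or a fallback of six chained AnimDynamics nodes on the dyn_chain bones.
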AnimBlueprint /BRCosmetics/Weapons/FORT_Melee/Pickaxe_FaunaPike_offhand/Meshes/FaunaPike_Offhand_Axe_AnimBP
target_skeleton=&/FaunaPike_Offhand_Axe_Skeleton
graphs [
 AnimGraph {nodes [
   n1 = "Transform (Modify) Bone - Bone: pelvis"(
    ComponentPose=Local To Component(LocalPose=Use cached pose 'InputPose'())
    Alpha=1
    Translation=Get pelvis_location() -> (pelvis_location:Vector)
    Rotation=Get pelvis_rotation() -> (pelvis_rotation:Rotator)
    Scale="1,1,1"
   ) {
    rotationMode=BMM_Replace
    rotationSpace=BCS_WorldSpace
    scaleSpace=BCS_BoneSpace
    translationMode=BMM_Replace
    translationSpace=BCS_WorldSpace
    alphaBoolBlend {blendOption=Linear}
    alphaScaleBiasClamp {inRange {max=0, min=-40}, outRange {max=0, min=1}}
    boneToModify {boneName=pelvis}
   }
   n2 = "Transform (Modify) Bone - Bone: thigh_l"(
    ComponentPose=@n1
    Alpha=1
    Translation=Get thigh_left_location() -> (thigh_left_location:Vector)
    Rotation=Get thigh_left_rotation() -> (thigh_left_rotation:Rotator)
    Scale="1,1,1"
   ) {
    rotationMode=BMM_Replace
    rotationSpace=BCS_WorldSpace
    scaleSpace=BCS_BoneSpace
    translationMode=BMM_Replace
    translationSpace=BCS_WorldSpace
    alphaBoolBlend {blendOption=Linear}
    alphaScaleBiasClamp {inRange {max=0, min=-40}, outRange {max=0, min=1}}
    boneToModify {boneName=thigh_l}
   }
   n3 = "Transform (Modify) Bone - Bone: thigh_r"(
    ComponentPose=@n2
    Alpha=1
    Translation=Get thigh_right_location() -> (thigh_right_location:Vector)
    Rotation=Get thigh_right_rotation() -> (thigh_right_rotation:Rotator)
    Scale="1,1,1"
   ) {
    rotationMode=BMM_Replace
    rotationSpace=BCS_WorldSpace
    scaleSpace=BCS_BoneSpace
    translationMode=BMM_Replace
    translationSpace=BCS_WorldSpace
    alphaBoolBlend {blendOption=Linear}
    alphaScaleBiasClamp {inRange {max=0, min=-40}, outRange {max=0, min=1}}
    boneToModify {boneName=thigh_r}
   }
   n4 = "Transform (Modify) Bone - Bone: calf_l"(
    ComponentPose=@n3
    Alpha=1
    Translation=Get calf_left_location() -> (calf_left_location:Vector)
    Rotation=Get calf_left_rotation() -> (calf_left_rotation:Rotator)
    Scale="1,1,1"
   ) {
    rotationMode=BMM_Replace
    rotationSpace=BCS_WorldSpace
    scaleSpace=BCS_BoneSpace
    translationMode=BMM_Replace
    translationSpace=BCS_WorldSpace
    alphaBoolBlend {blendOption=Linear}
    alphaScaleBiasClamp {inRange {max=0, min=-40}, outRange {max=0, min=1}}
    boneToModify {boneName=calf_l}
   }
   n5 = "Transform (Modify) Bone - Bone: calf_r"(
    ComponentPose=@n4
    Alpha=1
    Translation=Get calf_right_location() -> (calf_right_location:Vector)
    Rotation=Get calf_right_rotation() -> (calf_right_rotation:Rotator)
    Scale="1,1,1"
   ) {
    rotationMode=BMM_Replace
    rotationSpace=BCS_WorldSpace
    scaleSpace=BCS_BoneSpace
    translationMode=BMM_Replace
    translationSpace=BCS_WorldSpace
    alphaBoolBlend {blendOption=Linear}
    alphaScaleBiasClamp {inRange {max=0, min=-40}, outRange {max=0, min=1}}
    boneToModify {boneName=calf_r}
   }
   n6 = RigidBody(ComponentPose=@n5) {
    bClampLinearTranslationLimitToRefPose=true
    bOverrideWorldGravity=true
    overridePhysicsAsset=&/FaunaPike_Offhand_Axe_Physics
    baseBoneRef {boneName=Melee_TwoHanded_Handle}
    overrideWorldGravity {z=-980}
    simSpaceSettings {
     dampingAlpha=0
     maxAngularAcceleration=32
     maxAngularVelocity=8
     maxLinearAcceleration=1200
     maxLinearVelocity=700
     worldAlpha=1
     externalLinearDragV=<0.6 0.6 0.6>
    }
   }
   Output Pose(Result=Blend Poses by bool(BlendPose_0=Component To Local(ComponentPose=@n6), BlendPose_1=Use cached pose 'AnimDynamics'(), bActiveValue=Get bRigidBodyEnabled()) {blendType=Linear}) {layerGroup=None}
   AnimGraphNode_SubInput_0 = Input Pose()
   Save cached pose 'InputPose'(Pose=@AnimGraphNode_SubInput_0) {cacheName=InputPose}
   n7 = Anim Dynamics - Bone: dyn_chain_01(ComponentPose=Local To Component(LocalPose=Use cached pose 'InputPose'()), Alpha=1, ExternalForce="0,0,0") {
    angularDampingOverride=1
    angularSpringConstant=500
    bAngularSpring=true
    bOverrideAngularDamping=true
    bUseGravityOverride=true
    gravityScale=0.35
    simulationSpace=RootRelative
    boundBone {boneName=dyn_chain_01}
    componentAppliedLinearAccClamp=<3000 3000 3000>
    componentLinearAccScale=<0.6 0.6 0.6>
    componentLinearVelScale=<0.6 0.6 0.6>
    gravityOverride {z=-980}
    physicsBodyDefinitions [{
      boundBone {boneName=dyn_chain_01}
      boxExtents=<17 17 17>
      constraintSetup {angularLimitsMax=<0 76 0>, angularLimitsMin=<0 -75 0>, angularTarget=<1 0 0>}
      localJointOffset=<3 0 0>
     }]
    simSpaceSettings {maxAngularAcceleration=64, maxAngularVelocity=16, simSpaceAngularAlpha=1, externalAngularVelocity=<0.4 0.4 0.4>}
   }
   n8 = Anim Dynamics - Bone: dyn_chain_02(ComponentPose=@n7, Alpha=1, ExternalForce="0,0,0") {
    angularDampingOverride=1
    angularSpringConstant=500
    bAngularSpring=true
    bOverrideAngularDamping=true
    bUseGravityOverride=true
    gravityScale=0.35
    simulationSpace=RootRelative
    boundBone {boneName=dyn_chain_02}
    componentAppliedLinearAccClamp=<3000 3000 3000>
    componentLinearAccScale=<0.6 0.6 0.6>
    componentLinearVelScale=<0.6 0.6 0.6>
    gravityOverride {z=-980}
    physicsBodyDefinitions [{
      boundBone {boneName=dyn_chain_02}
      boxExtents=<16 16 16>
      constraintSetup {angularLimitsMax=<0 50 85>, angularLimitsMin=<0 -50 -85>, angularTarget=<1 0 0>}
      localJointOffset=<3 0 0>
     }]
    simSpaceSettings {maxAngularAcceleration=64, maxAngularVelocity=16, simSpaceAngularAlpha=1, externalAngularVelocity=<0.4 0.4 0.4>}
   }
   n9 = Anim Dynamics - Bone: dyn_chain_03(ComponentPose=@n8, Alpha=1, ExternalForce="0,0,0") {
    angularDampingOverride=1
    angularSpringConstant=500
    bAngularSpring=true
    bOverrideAngularDamping=true
    bUseGravityOverride=true
    gravityScale=0.35
    simulationSpace=RootRelative
    boundBone {boneName=dyn_chain_03}
    componentAppliedLinearAccClamp=<3000 3000 3000>
    componentLinearAccScale=<0.6 0.6 0.6>
    componentLinearVelScale=<0.6 0.6 0.6>
    gravityOverride {z=-980}
    physicsBodyDefinitions [{
      boundBone {boneName=dyn_chain_03}
      boxExtents=<15 15 15>
      constraintSetup {angularLimitsMax=<0 50 85>, angularLimitsMin=<0 -50 -85>, angularTarget=<1 0 0>}
      localJointOffset=<3 0 0>
     }]
    simSpaceSettings {maxAngularAcceleration=64, maxAngularVelocity=16, simSpaceAngularAlpha=1, externalAngularVelocity=<0.4 0.4 0.4>}
   }
   n10 = Anim Dynamics - Bone: dyn_chain_04(ComponentPose=@n9, Alpha=1, ExternalForce="0,0,0") {
    angularDampingOverride=1
    angularSpringConstant=500
    bAngularSpring=true
    bOverrideAngularDamping=true
    bUseGravityOverride=true
    gravityScale=0.35
    simulationSpace=RootRelative
    boundBone {boneName=dyn_chain_04}
    componentAppliedLinearAccClamp=<3000 3000 3000>
    componentLinearAccScale=<0.6 0.6 0.6>
    componentLinearVelScale=<0.6 0.6 0.6>
    gravityOverride {z=-980}
    physicsBodyDefinitions [{
      boundBone {boneName=dyn_chain_04}
      boxExtents=<14 14 14>
      constraintSetup {angularLimitsMax=<0 50 75>, angularLimitsMin=<0 -45 -70>, angularTarget=<1 0 0>}
      localJointOffset=<3 0 0>
     }]
    simSpaceSettings {maxAngularAcceleration=64, maxAngularVelocity=16, simSpaceAngularAlpha=1, externalAngularVelocity=<0.4 0.4 0.4>}
   }
   n11 = Anim Dynamics - Bone: dyn_chain_05(ComponentPose=@n10, Alpha=1, ExternalForce="0,0,0") {
    angularDampingOverride=1
    angularSpringConstant=500
    bAngularSpring=true
    bOverrideAngularDamping=true
    bUseGravityOverride=true
    gravityScale=0.35
    simulationSpace=RootRelative
    boundBone {boneName=dyn_chain_05}
    componentAppliedLinearAccClamp=<3000 3000 3000>
    componentLinearAccScale=<0.6 0.6 0.6>
    componentLinearVelScale=<0.6 0.6 0.6>
    gravityOverride {z=-980}
    physicsBodyDefinitions [{
      boundBone {boneName=dyn_chain_05}
      boxExtents=<13 13 13>
      constraintSetup {angularLimitsMax=<0 50 75>, angularLimitsMin=<0 -45 -70>, angularTarget=<1 0 0>}
      localJointOffset=<3 0 0>
     }]
    simSpaceSettings {maxAngularAcceleration=64, maxAngularVelocity=16, simSpaceAngularAlpha=1, externalAngularVelocity=<0.4 0.4 0.4>}
   }
   n12 = Anim Dynamics - Bone: dyn_chain_06(ComponentPose=@n11, Alpha=1, ExternalForce="0,0,0") {
    angularDampingOverride=1
    angularSpringConstant=500
    bAngularSpring=true
    bOverrideAngularDamping=true
    bUseGravityOverride=true
    gravityScale=0.35
    simulationSpace=RootRelative
    boundBone {boneName=dyn_chain_06}
    componentAppliedLinearAccClamp=<3000 3000 3000>
    componentLinearAccScale=<0.6 0.6 0.6>
    componentLinearVelScale=<0.6 0.6 0.6>
    gravityOverride {z=-980}
    physicsBodyDefinitions [{
      boundBone {boneName=dyn_chain_06}
      boxExtents=<12 12 12>
      constraintSetup {angularLimitsMax=<0 5 35>, angularLimitsMin=<0 -5 -35>, angularTarget=<1 0 0>}
      localJointOffset=<3 0 0>
     }]
    simSpaceSettings {maxAngularAcceleration=64, maxAngularVelocity=16, simSpaceAngularAlpha=1, externalAngularVelocity=<0.4 0.4 0.4>}
   }
   Save cached pose 'AnimDynamics'(Pose=Component To Local(ComponentPose=@n12)) {cacheName=AnimDynamics}
  ]}
 EventGraph {nodes [
   Event BlueprintUpdateAnimation()
   n1 = GetPartSkeletalMeshComponent()
   > Is Valid(InputObject=@n1)
   n2 = GetAttachParent!(self:Scene Component Object Reference=@n1) -> (ReturnValue:Scene Component Object Reference)
   > Is Valid(InputObject=@n2)
   > Set Body_Mesh(Body_Mesh:Scene Component Object Reference=@n2) -> (Output_Get:Scene Component Object Reference)
   > n3 = Sequence!(execute) -> (then_0, then_1, then_2:Exec, then_3:Exec, then_4:Exec, then_5:Exec)
   Set is On Pawn(execute=@n3.then_5, is On Pawn:bool=NOT Boolean(A=Get bIsOnPawn() -> (bIsOnPawn:bool))) -> (Output_Get:bool)
   n4 = Get Body_Mesh() -> (Body_Mesh:Scene Component Object Reference)
   n5 = BreakTransform(InTransform=GetSocketTransform(self=@n4, InSocketName=calf_r, TransformSpace=RTS_World))
   Set calf_right_location(execute=@n3.then_4, calf_right_location:Vector=@n5) -> (Output_Get:Vector)
   > Set calf_right_rotation(calf_right_rotation:Rotator=@n5.Rotation) -> (Output_Get:Rotator)
   n6 = BreakTransform(InTransform=GetSocketTransform(self=@n4, InSocketName=calf_l, TransformSpace=RTS_World))
   Set calf_left_location(execute=@n3.then_3, calf_left_location:Vector=@n6) -> (Output_Get:Vector)
   > Set calf_left_rotation(calf_left_rotation:Rotator=@n6.Rotation) -> (Output_Get:Rotator)
   n7 = BreakTransform(InTransform=GetSocketTransform(self=@n4, InSocketName=pelvis, TransformSpace=RTS_World))
   Set pelvis_location(execute=@n3.then_0, pelvis_location:Vector=@n7) -> (Output_Get:Vector)
   > Set pelvis_rotation(pelvis_rotation:Rotator=@n7.Rotation) -> (Output_Get:Rotator)
   n8 = BreakTransform(InTransform=GetSocketTransform(self=@n4, InSocketName=thigh_l, TransformSpace=RTS_World))
   Set thigh_left_location(execute=@n3.then_1, thigh_left_location:Vector=@n8) -> (Output_Get:Vector)
   > Set thigh_left_rotation(thigh_left_rotation:Rotator=@n8.Rotation) -> (Output_Get:Rotator)
   n9 = BreakTransform(InTransform=GetSocketTransform(self=@n4, InSocketName=thigh_r, TransformSpace=RTS_World))
   Set thigh_right_location(execute=@n3.then_2, thigh_right_location:Vector=@n9) -> (Output_Get:Vector)
   > Set thigh_right_rotation(thigh_right_rotation:Rotator=@n9.Rotation) -> (Output_Get:Rotator)
   TryGetPawnOwner()
  ]}
]
variables [
 Body_Mesh {type=SceneComponent, default=null}
 calf_left_location {type=Vector}
 calf_left_rotation {type=Rotator}
 calf_right_location {type=Vector}
 calf_right_rotation {type=Rotator}
 is On Pawn {type=boolean}
 pelvis_location {type=Vector}
 pelvis_rotation {type=Rotator}
 thigh_left_location {type=Vector}
 thigh_left_rotation {type=Rotator}
 thigh_right_location {type=Vector}
 thigh_right_rotation {type=Rotator}
]
